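# Translucent unlit material for Niagara particles, mainly meshes, in the Embers character cosmetic effects, with pendulum-style swinging.
# Fresnel, offset by panning fire noise, indexes a gradient texture for emissive color tinted by particle color.
# Opacity combines noise, Fresnel variants, optional UV and vertex masks, and depth fades.
# World position offset rotates UV-masked vertices around a dynamic-parameter pivot, using PreviousFrameSwitch for previous-frame positions, plus scalable noise wobble.
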Material /BRCosmetics/Effects/Fort_Effects/Effects/Characters/Athena_Parts/Embers/M_FX_Embers_PendulumMesh
graph {
 properties {
  bUsedWithNiagaraMeshParticles=true
  bUsedWithNiagaraSprites=true
  blendMode=BLEND_Translucent
  pixelDepthOffsetMode=PDOM_Legacy
  shadingModel=MSM_Unlit
  usedShadingModels=Unlit
 }
 nodes [
  n1 = ParticleColor()
  n2 = MaterialFunctionCall() {materialFunction=&MF_Fresnel_FX}
  n3 = Multiply(
   A=Power(Base=@n2.Fresnel, Exp=ScalarParameter() {defaultValue=1.5, group=BaseTexture, parameterName=Fresnel_Exp})
   B=ScalarParameter() {defaultValue=1, group=BaseTexture, parameterName=Fresnel_Brightness}
  )
  n4 = TextureCoordinate()
  n5 = Clamp(None=Multiply(
    A=ComponentMask(None=@n4) {g=true}
    B=ScalarParameter() {defaultValue=1, group=BaseTexture, parameterName=UVGrad-Mask_Multiply, sortPriority=34}
   ))
  n6 = Clamp(None=LinearInterpolate(A=@n3, B=Multiply(A=@n3, B=@n5), Alpha=ScalarParameter() {group=BaseTexture, parameterName=UVGrad-Mask_Toggle, sortPriority=33}))
  n7 = ScalarParameter() {defaultValue=1, group=DetailNoise, parameterName=DetailNoise-Color_Exponent}
  n8 = Multiply(
   A=TextureCoordinate()
   B=AppendVector(
    A=ScalarParameter() {defaultValue=1, group=DetailNoise, parameterName=Tiling-X_DetailNoise, sortPriority=34}
    B=ScalarParameter() {defaultValue=0.25, group=DetailNoise, parameterName=Tiling-Y_DetailNoise, sortPriority=34}
   )
  )
  n9 = LinearInterpolate(Alpha=ParticleRandom())
  n10 = Constant()
  n11 = Constant()
  n12 = LinearInterpolate(Alpha=ParticleRandom())
  n13 = Add(
   A=@n8
   B=LinearInterpolate(A=AppendVector(A=@n9, B=@n10), B=AppendVector(A=@n11, B=@n12), Alpha=ScalarParameter() {parameterName=UseParticleRandomUV_Select})
  )
  n14 = ScalarParameter() {parameterName=UseParticleRandomUV}
  n15 = ScalarParameter() {defaultValue=0.0661, group=DetailNoise, parameterName=PanSpeed-X_DetailNoise, sortPriority=34}
  n16 = ScalarParameter() {defaultValue=0.25, group=DetailNoise, parameterName=PanSpeed-Y_DetailNoise, sortPriority=34}
  n17 = TextureSampleParameter2D(UVs=Panner(Coordinate=LinearInterpolate(A=@n8, B=@n13, Alpha=@n14), Speed=AppendVector(A=@n15, B=@n16)) {speedX=0.0661, speedY=0.25}) {group=DetailNoise, parameterName=DetailNoise, texture=/Game/Global/Textures/Default/Noise/T_FireNoise_01}
  n18 = Add(
   A=ScalarParameter() {defaultValue=-0.05, group=BaseTexture, parameterName=ColorCoordinate-Offset}
   B=Multiply(A=Power(Base=@n17.R, Exp=@n7), B=ScalarParameter() {defaultValue=0.1, group=DetailNoise, parameterName=DetailNoise-Color_Influence})
  ) {constB=0}
  n19 = TextureSampleParameter2D(UVs=AppendVector(A=Constant(), B=Add(A=@n6, B=@n18))) {
   group=BaseTexture
   parameterName=Texture-Grad
   samplerSource=SSM_Clamp_WorldGroupSettings
   texture=&/T_VFX_Grad_EmbersSkin_01
  }
  n20 = VectorParameter() {group=BaseTexture, parameterName=Color_Intensity_Mult, defaultValue=<rgba 2 2 2 1>}
  MaterialFunctionCall(Emissive=Multiply(A=Multiply(A=@n19.RGB, B=@n20.RGB), B=@n1.RGB)) {materialFunction=&MF_FX_EmissiveAdjustments}
  n21 = MaterialFunctionCall(Float2=LinearInterpolate(
    A=TextureCoordinate()
    B=TextureCoordinate() {coordinateIndex=1}
    Alpha=ScalarParameter() {group=BaseTexture, parameterName=Use_UV_1, sortPriority=-1}
   )) {materialFunction=&BreakOutFloat2Components}
  n22 = Multiply(A=@n21.R, B=ScalarParameter() {defaultValue=1, group=MultipliedTexture_A, parameterName=MultipliedTexture_A_Tile_U, sortPriority=2})
  n23 = Multiply(A=@n21.G, B=ScalarParameter() {defaultValue=1, group=MultipliedTexture_A, parameterName=MultipliedTexture_A_Tile_V, sortPriority=3})
  n24 = ScalarParameter() {defaultValue=-0.45, group=MultipliedTexture_A, parameterName=MultipliedTexture_A_PanSpeed_X, sortPriority=1}
  n25 = ScalarParameter() {defaultValue=0.1, group=MultipliedTexture_A, parameterName=MultipliedTexture_A_PanSpeed_Y, sortPriority=2}
  n26 = TextureSampleParameter2D(UVs=Panner(Coordinate=AppendVector(A=@n22, B=@n23), Speed=AppendVector(A=@n24, B=@n25))) {
   group=MultipliedTexture_A
   parameterName=MultipliedTexture_A
   samplerSource=SSM_Wrap_WorldGroupSettings
   sortPriority=1
   texture=/BRCosmetics/Athena/Effects/Particles/Skydiving/NS_Hightower_Date/T_Hightower_Ribbon_Dots
  }
  Reroute(None=MaterialFunctionCall(XYZ=@n26.RGB) {materialFunction=&Swizzle}.YXZ)
  n27 = ScalarParameter() {defaultValue=1, group=DetailNoise, parameterName=DetailNoise-Opacity_Exponent}
  n28 = Clamp(None=Multiply(A=Power(Base=@n17.R, Exp=@n27), B=ScalarParameter() {defaultValue=6, group=DetailNoise, parameterName=DetailNoise-Opacity_Multiply}))
  n29 = MaterialFunctionCall(
   Brightness=ScalarParameter() {defaultValue=1, group=Opacity, parameterName=Opacity_Fresnel_Brightness}
   Exponent=ScalarParameter() {defaultValue=1, group=Opacity, parameterName=Opacity_Fresnel_Exp}
  ) {materialFunction=&MF_Fresnel_FX}
  n30 = LinearInterpolate(A=@n29.Fresnel, B=@n29.Fresnel, Alpha=ScalarParameter() {group=Opacity, parameterName=Opacity_Invert_Fresnel})
  n31 = ScalarParameter() {defaultValue=2, group=DetailNoise, parameterName=DetailNoise-Opacity_FresnelMask-Exp, sortPriority=33}
  n32 = Clamp(None=Multiply(
    A=Power(Base=@n2.Fresnel, Exp=@n31)
    B=ScalarParameter() {defaultValue=6, group=DetailNoise, parameterName=DetailNoise-Opacity_FresnelMask-Mult, sortPriority=33}
   ))
  n33 = LinearInterpolate(
   A=LinearInterpolate(A=Multiply(A=@n28, B=@n30), B=@n30, Alpha=@n32)
   B=Multiply(A=@n28, B=@n32)
   Alpha=ScalarParameter() {parameterName=UseMultipliedFresnel}
  )
  n34 = Multiply(
   A=Power(
    Base=MaterialFunctionCall() {materialFunction=&MF_Fresnel_FX}.Fresnel Inverted
    Exp=ScalarParameter() {defaultValue=1, parameterName=InvFresnel_Pow}
   )
   B=ScalarParameter() {defaultValue=1, parameterName=InvFresnel_Mult}
  )
  n35 = ScalarParameter() {parameterName=InvFresnel_Contrast}
  n36 = LinearInterpolate(
   A=@n33
   B=Multiply(
    A=@n33
    B=MaterialFunctionCall(In=@n34, Contrast=@n35) {materialFunction=&CheapContrast}.Result
   )
   Alpha=ScalarParameter() {parameterName=UseInvertedFresnel}
  ) {constA=1}
  n37 = Multiply(A=@n36, B=@n5)
  n38 = ScalarParameter() {group=BaseTexture, parameterName=UVGrad-Mask_Toggle, sortPriority=33}
  n39 = Constant() {r=1}
  n40 = ScalarParameter() {group=Opacity, parameterName=DepthFade_Distance}
  n41 = Multiply(A=@n1.A, B=Multiply(A=LinearInterpolate(A=@n36, B=@n37, Alpha=@n38), B=DepthFade(Opacity=@n39, FadeDistance=@n40)))
  n42 = Multiply(A=@n41, B=VertexColor().R)
  n43 = ScalarParameter() {parameterName=UseVertColorMult}
  n44 = Constant() {r=80}
  n45 = Constant() {r=30}
  Reroute(None=Multiply(
    A=LinearInterpolate(A=@n41, B=@n42, Alpha=@n43)
    B=MaterialFunctionCall(Fade Length=@n44, Fade Offset=@n45) {materialFunction=&CameraDepthFade}.Result
   )) {desc=Opacity}
  n46 = ScalarParameter() {defaultValue=1, parameterName=UsePendulum}
  n47 = DynamicParameter() {parameterIndex=1, paramNames [PrevPosX,PrevPosY,PrevPosZ,WPO_Scale]}
  n48 = Multiply(A=MaterialFunctionCall(R=@n47.PrevPosX, G=@n47.PrevPosX, B=@n47.PrevPosX) {materialFunction=&AppendMany}.RGB) {constB=-1}
  n49 = DynamicParameter() {parameterIndex=3, defaultValue=<rgba 20 20 1 0>, paramNames [PosX,PosY,PosZ,Rigidity]}
  n50 = MaterialFunctionCall(R=@n49.PosX, G=@n49.PosX, B=@n49.PosX) {materialFunction=&AppendMany}
  n51 = DynamicParameter() {parameterIndex=2, paramNames [MeshOriginX,MeshOriginY,MeshOriginZ,MaskFalloff]}
  n52 = MaterialFunctionCall(R=@n51.MeshOriginX, G=@n51.MeshOriginX, B=@n51.MeshOriginX, A=@n51.MeshOriginX) {materialFunction=&AppendMany}
  n53 = CrossProduct(A=@n48, B=Normalize(VectorInput=Subtract(A=@n50.RGB, B=@n52.RGB)))
  n54 = WorldPosition()
  n55 = LinearInterpolate(
   A=TextureCoordinate()
   B=TextureCoordinate() {coordinateIndex=1}
   Alpha=ScalarParameter() {defaultValue=1, group=WPO-Pendulum, parameterName=Use_UV_1_Pendulum, sortPriority=-1}
  )
  n56 = Subtract(A=OneMinus(None=ComponentMask(None=@n55) {g=true}), B=ScalarParameter() {group=WPO-Pendulum, parameterName=Mask Offset})
  n57 = Clamp(None=Power(Base=Clamp(None=@n56), Exp=@n51.MaskFalloff))
  n58 = DynamicParameter() {defaultValue=<rgba 1 1 1 0>, paramNames [PrevPosX,PrevPosY,PrevPosZ,Random]}
  n59 = MaterialFunctionCall(R=@n58.PrevPosX, G=@n58.PrevPosX, B=@n58.PrevPosX) {materialFunction=&AppendMany}
  n60 = CrossProduct(A=@n48, B=Normalize(VectorInput=Subtract(A=@n59.RGB, B=@n52.RGB)))
  n61 = WorldPosition()
  n62 = PreviousFrameSwitch(
   Current Frame=LinearInterpolate(A=RotateAboutAxis(NormalizedRotationAxis=@n53, RotationAngle=@n49.Rigidity, PivotPoint=@n52.RGB, Position=@n54.XYZ), Alpha=@n57) {constB=0}
   Previous Frame=LinearInterpolate(A=RotateAboutAxis(NormalizedRotationAxis=@n60, RotationAngle=@n49.Rigidity, PivotPoint=@n52.RGB, Position=@n61.XYZ), Alpha=@n57) {constB=0}
  )
  n63 = Multiply(
   A=TextureCoordinate()
   B=AppendVector(
    A=ScalarParameter() {defaultValue=1, group=WPO, parameterName=WPO_Noise_Tile_X, sortPriority=1}
    B=ScalarParameter() {defaultValue=1, group=WPO, parameterName=WPO_Noise_Tile_Y, sortPriority=2}
   )
  )
  n64 = AppendVector(
   A=ScalarParameter() {group=WPO, parameterName=WPO_Noise_Panner_X, sortPriority=3}
   B=ScalarParameter() {group=WPO, parameterName=WPO_Noise_Panner_Y, sortPriority=4}
  )
  n65 = Multiply(
   A=TextureSample(UVs=Panner(Coordinate=@n63, Speed=@n64)) {samplerType=SAMPLERTYPE_Normal, texture=/Game/Global/Textures/Default/Noise/T_Normal_Noise01}.RGB
   B=Multiply(
    A=ScalarParameter() {group=WPO, parameterName=WPO_Intensity, sortPriority=5}
    B=DynamicParameter() {parameterIndex=1, paramNames [PrevPosX,PrevPosY,PrevPosZ,WPO_Scale]}.WPO_Scale
   )
  )
  n66 = ComponentMask(None=TextureCoordinate()) {g=true}
  n67 = Clamp(None=Power(Base=OneMinus(None=@n66), Exp=ScalarParameter() {defaultValue=3, group=WPO, parameterName=WPO_Noise_Masking_Exp}))
  Reroute(None=Add(A=LinearInterpolate(B=@n62, Alpha=@n46), B=Multiply(A=@n65, B=@n67))) {desc=WPO}
 ]
}
